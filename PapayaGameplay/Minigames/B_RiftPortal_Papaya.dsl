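; Rift portal for the Papaya minigame that opens if its data row is enabled and teleports players to an endpoint offset from itself.
; Entering pawns that aren't DBNO get FX; the server teleports them after a delay, optionally triggering skydiving.
; An outer capsule drives a timeline that widens the rift material and particles, with expand/collapse audio.
; DestroyRift closes and hides the portal, fades looping audio, then destroys the actor after a hangtime.
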(event EventBeginPlay
  ;[1 1 1] EVENT--> BeginPlay
  (if (Class|AthenaFunctionLibrary|RowtoBool (Variables|Default|GetEnabledRow))
    (|CallParentFunction)
    (Variables|RiftPortal|SetPortalOpen true)
    (else
      (Gameplay|SilentDie false))))

(event Custom|PortalOpened
  (bind _sm_rift (Variables|Default|GetSM_Rift))
  ;[1 1 1] Toggle FX and animate shards after opening
  (Utilities|FlowControl|Sequence
    (:then_0
      (Components|Activation|SetActive (Variables|Default|GetNS_Rift_Idle_Loop) false false)
      (Components|Activation|SetActive (Variables|Default|GetNS_Rift_Portal_Closing) true false)
      (bind (_direction _open _glow) (Timeline|FragmentMovement Play)
        (:Update
          (Rendering|Material|SetScalarParameterValueonMaterials _sm_rift "OpenAdd" _open)
          (Rendering|Material|SetScalarParameterValueonMaterials _sm_rift "Emissive" _glow))))
    (:then_1
      (Utilities|FlowControl|Delay 1.0)
      (bind _returnvalue (Effects|Components|ParticleSystem|SpawnFXAtLocation (Variables|VFX|GetCrackExitRift) (+ (stale OnComponentBeginOverlap(OverlapCapsule) _output_get) "0, 0,200.000000") (Math|Rotator|MakeRotator 0.0 0.0 (Math|Random|RandomFloatinRange -1.8e+02 1.8e+02)) "1.000000,1.000000,1.000000" true "None" true))
      (LOD|SetMaxDrawDistance _returnvalue 0.0))))

(event Custom|PlayTeleportFX
  (params (Location :type Vector))
  ;[1 1 1] Spawn teleport visuals and audio for clients
  (Utilities|FlowControl|DoOnce false
    (:Completed
      (Effects|Components|ParticleSystem|SpawnFXAtLocation (Variables|VFX|GetEnterRift_NS) Location)
      (if (HelperFunctions|IsLocalPawn (Variables|RiftPortal|GetOverlapActor))
        (Audio|SpawnSoundAttached (Variables|Sound|GetEnteredRiftLocalPlayer) (Variables|/Script/Engine.Character|GetMesh (Variables|RiftPortal|GetPlayerPawn)))
        (else
          (Audio|PlaySoundatLocation (Variables|Sound|GetEnteredRift) Location))))))

(event Custom|ResetFX
  ;[1 1 1] Spawn teleport visuals and audio for clients
  (Utilities|FlowControl|DoOnce false
    (:Completed
      (Effects|Components|ParticleSystem|SpawnFXAtLocation (Variables|VFX|GetEnterRift_NS) (stale Custom|PlayTeleportFX Location))
      (if (HelperFunctions|IsLocalPawn (Variables|RiftPortal|GetOverlapActor))
        (Audio|SpawnSoundAttached (Variables|Sound|GetEnteredRiftLocalPlayer) (Variables|/Script/Engine.Character|GetMesh (Variables|RiftPortal|GetPlayerPawn)))
        (else
          (Audio|PlaySoundatLocation (Variables|Sound|GetEnteredRift) (stale Custom|PlayTeleportFX Location)))))))

(event OnComponentBeginOverlap(OverlapCapsule)
  (params (OverlappedComponent :type Primitive%20Component%20Object%20Reference) (OtherActor :type Actor%20Object%20Reference) (OtherComp :type Primitive%20Component%20Object%20Reference) (OtherBodyIndex :type Integer) (bFromSweep :type Boolean) (SweepResult :type Hit%20Result%20Structure%20%28by%20ref%29))
  (bind _overlapactor (Variables|RiftPortal|GetOverlapActor))
  (bind (_endpointtransform_location _endpointtransform_rotation) (Math|Transform|BreakTransform (Variables|GetEndPointTransform_Location)))
  (bind (_endpointtransform_rotation_roll _endpointtransform_rotation_pitch _endpointtransform_rotation_yaw) (Math|Rotator|BreakRotator _endpointtransform_rotation))
  (bind _endpointtransform_location_2 (Math|Transform|BreakTransform (Variables|GetEndPointTransform_Location)))
  ;[1 1 1] Teleport Player
  (Variables|RiftPortal|SetOverlapActor OtherActor)
  (if (Variables|RiftPortal|GetPortalOpen)
    (Variables|RiftPortal|SetActorRotation (Transformation|GetActorRotation (Variables|RiftPortal|GetOverlapActor)))
    (bind _output_get (Variables|RiftPortal|SetTeleportLocation (+ _endpointtransform_location_2 (Transformation|GetActorLocation))))
    (CallFunction|ResetFX)
    (Utilities|IsValid _overlapactor
      (:Is%20Valid
        (bind _asfort_player_pawn (Utilities|Casting|CastToFortPlayerPawn _overlapactor))
        (if (not (Variables|/Script/FortniteGame.FortPawn|GetbIsDBNO _asfort_player_pawn))
          (Variables|RiftPortal|SetPlayerPawn _asfort_player_pawn)
          (CallFunction|PlayTeleportFX (Transformation|GetActorLocation _asfort_player_pawn))
          (if (Vehicle|IsinVehicle _asfort_player_pawn)
            (if (not (Vehicle|CanTeleport (Vehicle|GetVehicleInterface _asfort_player_pawn)))
              (GameplayEffects|RemoveActiveEffectswithGrantedTags (Ability|GetAbilitySystemComponent _asfort_player_pawn) "(GameplayTags=((TagName=\"Gameplay.Action.Player.AthenaInVehicle\")))") ; If vehicle isn't supposed to teleport then kick the player from it
            ))
          (Networking|SwitchHasAuthority
            (:Authority
              (RemoteControlPawn|ForceDetachfromRemoteControlledPawn _asfort_player_pawn)
              (Variables|/Script/FortniteGame.FortPawn|SetbIgnoreNextFallingDamage true _asfort_player_pawn) ; Have to do this because of the delay. Otherwise player can die from fall damage if they fall from high distance.
              (Utilities|FlowControl|Delay (Variables|RiftPortal|GetTeleportDelay))
              (bind _returnvalue (CallFunction|IsFlying))
              (if (not _returnvalue)
                (Parachute|EndSkydiving _asfort_player_pawn) ; End Skydiving if player was previously
                (Components|Movement|StopMovementImmediately (Variables|/Script/Engine.Character|GetCharacterMovement _asfort_player_pawn))
                (Transform|Actor|TeleportPlayerPawn _asfort_player_pawn (Variables|RiftPortal|GetTeleportLocation) (Math|Rotator|MakeRotator 0.0 0.0 _endpointtransform_rotation_yaw) true false)
                (if (Variables|RiftPortal|GetSkyDiveAfterTeleport)
                  (Ability|SendGameplayEventtoActor _asfort_player_pawn "(TagName=\"Abilities.Generic.Rift.Trigger\")"))))))))))

(event EventOverlappingOnBeginPlay
  (params (Actor :type Actor%20Object%20Reference))
  (bind _overlapactor (Variables|RiftPortal|GetOverlapActor))
  (bind (_endpointtransform_location _endpointtransform_rotation) (Math|Transform|BreakTransform (Variables|GetEndPointTransform_Location)))
  (bind (_endpointtransform_rotation_roll _endpointtransform_rotation_pitch _endpointtransform_rotation_yaw) (Math|Rotator|BreakRotator _endpointtransform_rotation))
  (bind _endpointtransform_location_2 (Math|Transform|BreakTransform (Variables|GetEndPointTransform_Location)))
  ;[1 1 1] Teleport Player
  (Variables|RiftPortal|SetOverlapActor Actor)
  (if (Variables|RiftPortal|GetPortalOpen)
    (Variables|RiftPortal|SetActorRotation (Transformation|GetActorRotation (Variables|RiftPortal|GetOverlapActor)))
    (bind _output_get (Variables|RiftPortal|SetTeleportLocation (+ _endpointtransform_location_2 (Transformation|GetActorLocation))))
    (CallFunction|ResetFX)
    (Utilities|IsValid _overlapactor
      (:Is%20Valid
        (bind _asfort_player_pawn (Utilities|Casting|CastToFortPlayerPawn _overlapactor))
        (if (not (Variables|/Script/FortniteGame.FortPawn|GetbIsDBNO _asfort_player_pawn))
          (Variables|RiftPortal|SetPlayerPawn _asfort_player_pawn)
          (CallFunction|PlayTeleportFX (Transformation|GetActorLocation _asfort_player_pawn))
          (if (Vehicle|IsinVehicle _asfort_player_pawn)
            (if (not (Vehicle|CanTeleport (Vehicle|GetVehicleInterface _asfort_player_pawn)))
              (GameplayEffects|RemoveActiveEffectswithGrantedTags (Ability|GetAbilitySystemComponent _asfort_player_pawn) "(GameplayTags=((TagName=\"Gameplay.Action.Player.AthenaInVehicle\")))") ; If vehicle isn't supposed to teleport then kick the player from it
            ))
          (Networking|SwitchHasAuthority
            (:Authority
              (RemoteControlPawn|ForceDetachfromRemoteControlledPawn _asfort_player_pawn)
              (Variables|/Script/FortniteGame.FortPawn|SetbIgnoreNextFallingDamage true _asfort_player_pawn) ; Have to do this because of the delay. Otherwise player can die from fall damage if they fall from high distance.
              (Utilities|FlowControl|Delay (Variables|RiftPortal|GetTeleportDelay))
              (bind _returnvalue (CallFunction|IsFlying))
              (if (not _returnvalue)
                (Parachute|EndSkydiving _asfort_player_pawn) ; End Skydiving if player was previously
                (Components|Movement|StopMovementImmediately (Variables|/Script/Engine.Character|GetCharacterMovement _asfort_player_pawn))
                (Transform|Actor|TeleportPlayerPawn _asfort_player_pawn (Variables|RiftPortal|GetTeleportLocation) (Math|Rotator|MakeRotator 0.0 0.0 _endpointtransform_rotation_yaw) true false)
                (if (Variables|RiftPortal|GetSkyDiveAfterTeleport)
                  (Ability|SendGameplayEventtoActor _asfort_player_pawn "(TagName=\"Abilities.Generic.Rift.Trigger\")"))))))))))

(event Custom|DestroyRift
  (bind _audio_looping (Variables|Sound|GetAudio_Looping))
  ;[1 1 1] On Death
  (Variables|RiftPortal|SetPortalOpen false)
  (Utilities|FlowControl|Sequence
    (:then_0
      (Networking|CanExecuteCosmeticEvents
        (:True
          (Development|SetHiddeninGame (Variables|Default|GetSM_Rift) true false)
          (Audio|PlaySoundatLocation (Variables|Sound|GetRiftDestroyed) (Transformation|GetActorLocation)) ; Play Deactivate Sound
          (bind (_direction _pitchcurve) (Timeline|Audio%20RiftDestroyed Play)
            (:Update
              (Audio|Components|Audio|SetPitchMultiplier _audio_looping _pitchcurve))
            (:Finished
              (Audio|Components|Audio|FadeOut _audio_looping 1 0.0 "Linear" :!class /Script/Engine.AudioComponent))))))
    (:then_1
      (Utilities|FlowControl|Delay (+ (ScalableFloat|GetValueatLevel (Variables|RiftPortal|GetHangtime)) 1.0))
      (Actor|DestroyActor))))

(event OnComponentBeginOverlap(Capsule)
  (params (OverlappedComponent :type Primitive%20Component%20Object%20Reference) (OtherActor :type Actor%20Object%20Reference) (OtherComp :type Primitive%20Component%20Object%20Reference) (OtherBodyIndex :type Integer) (bFromSweep :type Boolean) (SweepResult :type Hit%20Result%20Structure%20%28by%20ref%29))
  (bind _sm_rift (Variables|Default|GetSM_Rift))
  ;[1 1 1] On Overlap - Toggle Post Process
  (Utilities|Casting|CastToFortPlayerPawn OtherActor)
  (CallFunction|PlayExpandorCollapseAudio true)
  ;;[1 1 1] Widen rift if player is close
  (bind (_direction _open) (Timeline|VectorScale Play)
    (:Update
      (Rendering|Material|SetScalarParameterValueonMaterials _sm_rift "Open" (Math|Float|SelectFloat 0.0 _open (Components|Timeline|IsPlaying (Variables|B_RiftPortal_Papaya|GetFragmentMovement))))
      (Transformation|SetRelativeRotation _sm_rift (Math|Rotator|MakeRotator (* _open -2e+01)) false false)
      (Effects|Components|ParticleSystem|SetFloatParameter (Variables|Default|GetNS_Rift_Idle_Loop) "Open" _open))))

(event OnComponentEndOverlap(Capsule)
  (params (OverlappedComponent :type Primitive%20Component%20Object%20Reference) (OtherActor :type Actor%20Object%20Reference) (OtherComp :type Primitive%20Component%20Object%20Reference) (OtherBodyIndex :type Integer))
  (bind _sm_rift (Variables|Default|GetSM_Rift))
  ;[1 1 1] On Overlap - Toggle Post Process
  (if (<= (Utilities|Array|Length (Collision|GetOverlappingActors OtherActor "/Script/FortniteGame.FortPlayerPawn")) 0)
    (CallFunction|PlayExpandorCollapseAudio false)
    ;;[1 1 1] Widen rift if player is close
    (Timeline|VectorScale Reverse)))

(fn IsFlying
  (bind _playerpawn (Variables|RiftPortal|GetPlayerPawn))
  (Utilities|IsValid _playerpawn
    (:Is%20Valid
      (bind _asfort_movement_comp_character (Utilities|Casting|CastToFortMovementComp_Character (Pawn|GetMovementComponent _playerpawn))
        (:then
          (return (Utilities|Enum|Equal(Enum) (Variables|/Script/Engine.CharacterMovementComponent|GetMovementMode _asfort_movement_comp_character) "MOVE_Flying")))
        (:CastFailed
          (return false))))))

(fn PlayExpandOrCollapseAudio
  (params (Expand :type Boolean))
  (if (Variables|RiftPortal|GetPortalOpen)
    (Audio|PlaySoundatLocation (select Expand "/Game/Sounds/Fort_GamePlay_Sounds/Player/RiftPortal/RiftPortal_Expand_Cue.RiftPortal_Expand_Cue" "/Game/Sounds/Fort_GamePlay_Sounds/Player/RiftPortal/RiftPortal_Collapse_Cue.RiftPortal_Collapse_Cue") (Transformation|GetWorldLocation (Variables|Default|GetSM_Rift)))))

(fn ShouldPlay1PAudio
  (params (Target :type Pawn%20Object%20Reference))
  (bind _returnvalue (Fortnite|GetLocalFortPlayerControllers false true))
  (bind _output (Utilities|Array|Get(acopy) _returnvalue))
  (if (> (Utilities|Array|Length _returnvalue) 0)
    (Utilities|IsValid _output
      (:Is%20Valid
        (return (Utilities|Equal(Object) (Pawn|GetViewTarget _output) Target)))
      (:Is%20Not%20Valid
        _))
    (else
      _))
  (return (Pawn|IsLocallyControlled Target)))

(fn ConstructionScript)
